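# Temporary translucent UI material for a ToyFox minimap icon with an animated, tinted pulse behind a scalable icon texture.
# Two circle pulses, with start and end sizes blended through reroutes for direction flipping, are clipped by a hexagon mask.
# They combine with static and time-animated hexagon glows and the icon alpha into one mask.
# PulseColor, brightened toward the center by a circle gradient, tints that mask before the icon's RGB is blended in.
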Material /ToyFoxGameplay/Assets/Material/TEMP/M_UI_TEMP_ToyFox_MinimapIcon
graph {
 properties {blendMode=BLEND_Translucent, materialDomain=MD_UI, pixelDepthOffsetMode=PDOM_Legacy, usedShadingModels=Default Lit}
 nodes [
  n1 = VectorParameter() {parameterName=PulseColor, defaultValue=<rgba 0.8308 0 0.0222 1>}
  n2 = MaterialFunctionCall(UVs=TextureCoordinate(), "Rotation Angle (0-1)"=Constant() {r=0.25}) {materialFunction=&CustomRotator}
  n3 = Constant() {r=0.05}
  n4 = Constant() {r=0.5}
  n5 = LinearInterpolate(
   A=@n1.RGB
   B=Add(A=@n1.RGB) {constB=0.1}
   Alpha=Multiply(A=MaterialFunctionCall(UVs=@n2.Rotated Values, Size=@n3, GlowMax=@n4) {materialFunction=&MF_UI_Circle}.Fill) {constB=0.5}
  )
  n6 = TextureSampleParameter2D(UVs=MaterialFunctionCall(UVs=TextureCoordinate(), Texture Scale=ScalarParameter() {defaultValue=0.2, group=Icon, parameterName=IconScale}) {materialFunction=&ScaleUVsByCenter}.UVs) {
   group=Icon
   parameterName=IconTexture
   samplerSource=SSM_Clamp_WorldGroupSettings
   texture=/ToyFoxGameplay/UI/Textures/T_UI_ToyFox_Minimap_Icon
  }
  n7 = ScalarParameter() {defaultValue=1.1, group=PulseStyle, parameterName=ScaleEnd}
  n8 = ScalarParameter() {group=PulseStyle, parameterName="ScaleStart (S)"}
  n9 = NamedRerouteUsage()
  n10 = Constant() {r=0.25}
  n11 = ScalarParameter() {defaultValue=1, group=PulseStyle, parameterName=ScaleEnd2}
  n12 = NamedRerouteUsage()
  n13 = Max(
   A=MaterialFunctionCall!(
    PulseSpeed
    SizeEnd=LinearInterpolate(A=@n7, B=@n8, Alpha=@n9)
    WidthStart
    WidthEnd=ScalarParameter() {defaultValue=0.2, group=PulseStyle, parameterName="WidthEnd (S)"}
    PulseAnimFalloff
    SizeStart=LinearInterpolate(A=@n8, B=@n7, Alpha=@n9)
    OpacityStart
    OpacityEnd=@n10
    TimeSpeed
    TexCoord
   ) {materialFunction=&/MF_CirclePulse}.Result
   B=MaterialFunctionCall(
    SizeEnd=LinearInterpolate(A=@n11, B=@n8, Alpha=@n12)
    WidthEnd=ScalarParameter() {defaultValue=0.01, group=PulseStyle, parameterName=WidthEnd2}
    PulseAnimFalloff=ScalarParameter() {defaultValue=0.2605, group=PulseStyle, parameterName="PulseAnimFalloff (S)"}
    SizeStart=LinearInterpolate(A=@n8, B=@n7, Alpha=@n12)
    OpacityEnd=@n10
   ) {materialFunction=&MF_CirclePulse}.Result
  )
  n14 = Constant() {r=6}
  n15 = MaterialFunctionCall!(
   UVs=MaterialFunctionCall(UVs=TextureCoordinate(), "Rotation Angle (0-1)"=Constant() {r=0.25}) {materialFunction=&CustomRotator}.Rotated Values
   Size=Constant() {r=0.5}
   Sides=@n14
   InnerAngle
   OutlineThickness
   OutlineAlignment
   GlowMin
   GlowMax=Constant() {r=0.4}
  ) {materialFunction=/FortUILibrary/MaterialFunctions/Shapes/MF_UI_RegularPolygon}
  n16 = TextureCoordinate()
  n17 = Constant() {r=0.25}
  n18 = Constant() {r=0.08}
  n19 = Time()
  n20 = Multiply(A=Cosine(None=Frac(None=@n19))) {constB=0.1}
  n21 = Constant() {r=0.2}
  n22 = TextureCoordinate()
  n23 = Constant() {r=0.25}
  n24 = Add(
   A=MaterialFunctionCall(
    UVs=MaterialFunctionCall(UVs=@n16, "Rotation Angle (0-1)"=@n17) {materialFunction=&CustomRotator}.Rotated Values
    Size=Add(A=@n18, B=@n20)
    Sides=@n14
    GlowMax=Add(A=@n20, B=@n21)
   ) {materialFunction=&MF_UI_RegularPolygon}.Fill
   B=MaterialFunctionCall(
    UVs=MaterialFunctionCall(UVs=@n22, "Rotation Angle (0-1)"=@n23) {materialFunction=&CustomRotator}.Rotated Values
    Size=Constant() {r=0.08}
    Sides=Constant() {r=6}
    GlowMax=Constant() {r=0.2}
   ) {materialFunction=&MF_UI_RegularPolygon}.Fill
  )
  n25 = Max(A=@n6.A, B=Max(A=Min(A=@n13, B=@n15.Fill), B=Multiply(A=@n24) {constB=0.5}))
  LinearInterpolate(A=LinearInterpolate(A=Multiply(A=@n5, B=@n25), Alpha=@n6.A) {constB=0}, B=@n6.RGB, Alpha=@n6.RGB) {constB=0}
  n26 = ScalarParameter() {parameterName=FlipPulseDirection}
  NamedRerouteDeclaration(None=Saturate(None=Ceil(None=@n26))) {name=P_FlipPulseDirection, nodeColor=<rgba 0.4 0.4 0.4 1>}
  OneMinus(None=Saturate(None=Add(A=@n6.RGB, B=@n6.RGB) {constB=0}))
 ]
}
